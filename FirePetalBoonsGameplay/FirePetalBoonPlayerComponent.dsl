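; Player controller component for Fire Petal boons that keeps the FirePetalNotDBNO loose gameplay tags in step with the player's downed (DBNO) state.
; On BeginPlay it caches the owning Fort controller, then handles its pawn immediately or binds to possessed-pawn changes if none is valid.
; For that pawn, parallel ability tasks watch Gameplay.Status.DBNO, removing the tags when it is added and adding them back when it is removed.
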(event EventBeginPlay
  (bind _asfort_player_controller (Utilities|Casting|CastToFortPlayerController (Components|GetOwner)))
  (bind _output_get (Variables|Default|SetFort%20Player%20Controller _asfort_player_controller))
  (Utilities|IsValid (Game|Player|GetPlayerPawn _output_get)
    (:Is%20Valid
      (CallFunction|OnPlayerPawnPossessedChanged 0 (Game|Player|GetPlayerPawn _output_get)))
    (:Is%20Not%20Valid
      (Pawn|AssignOnPossessedPawnChanged (Variables|Default|GetFort%20Player%20Controller) (delegate Custom|OnPlayerPawnPossessedChanged)))))

(event EventTick
  (params (DeltaSeconds :type Float%20%28single-precision%29)))

(event Custom|OnPlayerPawnPossessedChanged
  (params (OldPawn :type Pawn%20Object%20Reference) (NewPawn :type Pawn%20Object%20Reference))
  (bind _asfort_player_pawn (Utilities|Casting|CastToFortPlayerPawn NewPawn))
  (bind _output_get (Variables|Default|SetFort%20Player%20Pawn _asfort_player_pawn))
  (Ability|Tasks|WaitGameplayTagAddToActor _output_get "(TagName=\"Gameplay.Status.DBNO\")"
    (:then
      (Ability|Tasks|WaitGameplayTagRemoveFromActor _output_get "(TagName=\"Gameplay.Status.DBNO\")"
        (:Removed
          (Ability|GameplayEffect|AddLooseGameplayTags _output_get (Variables|Default|GetFirePetalNotDBNO)))))
    (:Added
      (Ability|GameplayEffect|RemoveLooseGameplayTags _output_get (Variables|Default|GetFirePetalNotDBNO)))))
